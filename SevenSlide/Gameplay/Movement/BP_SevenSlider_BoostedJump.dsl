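; Boosted-jump logic for the SevenSlider movement mode, with a function that picks a directional jump montage for the owning pawn.
; GetAnimationMontage returns nothing if any montage is already playing on the pawn's mesh.
; Otherwise it dots normalized velocity against the pawn's forward, right, left and back vectors, keeping the best-aligned direction.
; It returns the matching forward, right, left or back jump montage, with forward as the starting choice.
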(fn GetAnimationMontage
  (bind _ownerpawn (Variables|FortMovementMode_BaseExtLogic|GetOwnerPawn))
  (bind _returnvalue (Transformation|GetActorRightVector _ownerpawn))
  (bind _returnvalue_2 (Math|Vector|DotProduct (Math|Vector|Normalize (Transformation|GetVelocity (Variables|FortMovementMode_BaseExtLogic|GetOwnerPawn))) _returnvalue))
  (bind _ownerpawn_2 (Variables|FortMovementMode_BaseExtLogic|GetOwnerPawn))
  (bind _returnvalue_3 (Transformation|GetActorRightVector _ownerpawn_2))
  (bind _returnvalue_4 (* _returnvalue_3 -1.0))
  (bind _returnvalue_5 (Math|Vector|DotProduct (Math|Vector|Normalize (Transformation|GetVelocity (Variables|FortMovementMode_BaseExtLogic|GetOwnerPawn))) _returnvalue_4))
  (bind _ownerpawn_3 (Variables|FortMovementMode_BaseExtLogic|GetOwnerPawn))
  (bind _returnvalue_6 (Transformation|GetVelocity _ownerpawn_3))
  (bind _returnvalue_7 (Math|Vector|Normalize _returnvalue_6))
  (if (Animation|Montage|IsAnyMontagePlaying (Components|SkeletalMesh|GetAnimInstance (Variables|/Script/Engine.Character|GetMesh (Variables|FortMovementMode_BaseExtLogic|GetOwnerPawn))))
    (return 0)
    (else
      (LocalVariables|SetLocalVarMontage (Variables|Default|GetForwardMontage_Jump))
      (LocalVariables|SetLocalMostAlignedValue (Math|Vector|DotProduct _returnvalue_7 (Transformation|GetActorForwardVector (Variables|FortMovementMode_BaseExtLogic|GetOwnerPawn))))
      (if (> _returnvalue_2 (LocalVariables|GetLocalMostAlignedValue))
        (LocalVariables|SetLocalVarMontage (Variables|Default|GetRightMontage_Jump))
        (LocalVariables|SetLocalMostAlignedValue _returnvalue_2)
        (else
          _))
      (if (> _returnvalue_5 (LocalVariables|GetLocalMostAlignedValue))
        (LocalVariables|SetLocalVarMontage (Variables|Default|GetLeftMontage_Jump))
        (LocalVariables|SetLocalMostAlignedValue _returnvalue_5)
        (else
          _))
      (if (> (Math|Vector|DotProduct _returnvalue_7 (* (Transformation|GetActorForwardVector (Variables|FortMovementMode_BaseExtLogic|GetOwnerPawn)) -1.0)) (LocalVariables|GetLocalMostAlignedValue))
        (LocalVariables|SetLocalVarMontage (Variables|Default|GetBackMontage_Jump)))
      (return (LocalVariables|GetLocalVarMontage)))))
