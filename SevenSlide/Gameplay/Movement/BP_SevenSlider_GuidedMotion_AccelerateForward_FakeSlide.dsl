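; Movement mode extension for the SevenSlide fake-slide guided motion that, by its name, accelerates the player forward; only its deactivation cleanup is defined here.
; On deactivate it uncrouches the owner pawn through both the Character and FortPlayerPawn crouch calls.
; It then sets character movement to Falling and applies the configured "Effects to Apply" gameplay effects to the pawn at its location.
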(event EventOnExtensionDeactivate
  (params (RuntimeContext :type MMERuntime%20Context%20Structure%20%28by%20ref%29))
  (bind _ownerpawn (Variables|FortMovementMode_BaseExtLogic|GetOwnerPawn))
  (Character|UnCrouch _ownerpawn)
  (bind _asfort_player_pawn (Utilities|Casting|CastToFortPlayerPawn _ownerpawn))
  (Class|FortPlayerPawn|SetCrouch _asfort_player_pawn false false)
  (Pawn|Components|CharacterMovement|SetMovementMode (Variables|/Script/Engine.Character|GetCharacterMovement _asfort_player_pawn) "MOVE_Falling" 0)
  (Fortnite|ApplyGameplayEffectsToActor (Variables|FortMovementMode_BaseExtLogic|GetOwnerPawn) (Variables|Default|GetEffects%20to%20Apply) (Transformation|GetActorLocation (Variables|FortMovementMode_BaseExtLogic|GetOwnerPawn))))
