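; Seven Slide guided-motion movement mode extension that manages a camera modifier and turns off pawn movement dust while active.
; On activation, behind a hotfix scalable float, it removes and re-adds the modifier; deactivation disables it and restores dust.
; SetCameraModifier acts only without authority, on local viewing controllers' camera managers.
; A second hotfix selects a revised flow that tracks modified managers for later cleanup, or the deprecated per-controller flow.
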(event EventOnExtensionActivate
  (params (RuntimeContext :type MMERuntime%20Context%20Structure%20%28by%20ref%29))
  (if (> (ScalableFloat|GetValueasDouble (Variables|Default|GetCameraHotfixEnabled)) 0)
    (Default|SetCameraModifier false true)
    (Default|SetCameraModifier true)
    (else
      _))
  (bind _asfort_player_pawn (Utilities|Casting|CastToFortPlayerPawn (Variables|FortMovementMode_BaseExtLogic|GetOwnerPawn)))
  (Variables|/Script/FortniteGame.FortPlayerPawn|SetbFXPlayDustOnMovement false _asfort_player_pawn))

(event EventOnExtensionDeactivate
  (params (RuntimeContext :type MMERuntime%20Context%20Structure%20%28by%20ref%29))
  (if (> (ScalableFloat|GetValueasDouble (Variables|Default|GetCameraHotfixEnabled)) 0)
    (Default|SetCameraModifier false)
    (else
      _))
  (bind _asfort_player_pawn (Utilities|Casting|CastToFortPlayerPawn (Variables|FortMovementMode_BaseExtLogic|GetOwnerPawn)))
  (Variables|/Script/FortniteGame.FortPlayerPawn|SetbFXPlayDustOnMovement true _asfort_player_pawn))

(fn SetCameraModifier
  (params (Enabled? :type Boolean) (Remove? :type Boolean))
  (if (Networking|HasAuthority (Variables|FortMovementMode_BaseExtLogic|GetOwnerPawn))
    _
    (elif (ScalableFloat|GetValueasBool (Variables|Default|GetHF_UseRevisedCameraModifierFlow))
      ; Revised Modifier Flow
      (Variables|GetEnabled?
        (:then
          (Utilities|Array|Clear (Variables|Default|GetCameraManagersWithAppliedModifiers))
          (for _array_element (FortPawn|GetAllLocalViewingPlayerControllers (Variables|FortMovementMode_BaseExtLogic|GetOwnerPawn))
            (Utilities|IsValid _array_element
              (:Is%20Valid
                (bind _playercameramanager (Variables|/Script/Engine.PlayerController|GetPlayerCameraManager _array_element)
                  (:then
                    (bind _output_get (LocalVariables|SetPlayerCameraManager _playercameramanager))
                    (bind _returnvalue (Game|Player|FindCameraModifierbyClass _output_get (Variables|Default|GetCameraModifierCurrent)))
                    (Utilities|IsValid _returnvalue
                      (:Is%20Not%20Valid
                        (bind _returnvalue_2 (Game|Player|AddNewCameraModifier (LocalVariables|GetPlayerCameraManager) (Variables|Default|GetCameraModifierCurrent)))
                        (Utilities|IsValid _returnvalue_2
                          (:Is%20Valid
                            (CameraModifier|EnableModifier _returnvalue_2)
                            (Utilities|Array|Add (Variables|Default|GetCameraManagersWithAppliedModifiers) (LocalVariables|GetPlayerCameraManager)))))))
                  (:else))))))
        (:else
          (for _array_element_2 (Variables|Default|GetCameraManagersWithAppliedModifiers)
            (Utilities|IsValid _array_element_2
              (:Is%20Valid
                (bind _output_get_2 (LocalVariables|SetPlayerCameraManager _array_element_2))
                (bind _returnvalue_3 (Game|Player|FindCameraModifierbyClass _output_get_2 (Variables|Default|GetCameraModifierCurrent)))
                (Utilities|IsValid _returnvalue_3
                  (:Is%20Valid
                    (CameraModifier|DisableModifier _returnvalue_3 false)
                    (Variables|GetRemove?
                      (:then
                        ;;[1 1 1] This stops the camera from being weird at start - fix this on the camera modifier if time allows
                        (Game|Player|RemoveCameraModifier (LocalVariables|GetPlayerCameraManager) _returnvalue_3))
                      (:else)))))))))
      (else
        ; Deprecated Modifier Flow
        (for _array_element_3 (FortPawn|GetAllLocalViewingPlayerControllers (Variables|FortMovementMode_BaseExtLogic|GetOwnerPawn))
          (Utilities|IsValid _array_element_3
            (:Is%20Valid
              (bind _playercameramanager_2 (Variables|/Script/Engine.PlayerController|GetPlayerCameraManager _array_element_3)
                (:then
                  (LocalVariables|SetPlayerCameraManager _playercameramanager_2)
                  ;;[1 1 1] Main Player
                  (Variables|GetEnabled?
                    (:then
                      (bind _returnvalue_4 (Game|Player|FindCameraModifierbyClass (LocalVariables|GetPlayerCameraManager) (Variables|Default|GetCameraModifierCurrent)))
                      (Utilities|IsValid _returnvalue_4
                        (:Is%20Not%20Valid
                          (bind _returnvalue_5 (Game|Player|AddNewCameraModifier (LocalVariables|GetPlayerCameraManager) (Variables|Default|GetCameraModifierCurrent)))
                          (CameraModifier|EnableModifier _returnvalue_5))))
                    (:else
                      (bind _returnvalue_6 (Game|Player|FindCameraModifierbyClass (LocalVariables|GetPlayerCameraManager) (Variables|Default|GetCameraModifierCurrent)))
                      (Utilities|IsValid _returnvalue_6
                        (:Is%20Valid
                          (CameraModifier|DisableModifier _returnvalue_6 false)
                          (Variables|GetRemove?
                            (:then
                              ;;;[1 1 1] This stops the camera from being weird at start - fix this on the camera modifier if time allows
                              (Game|Player|RemoveCameraModifier (LocalVariables|GetPlayerCameraManager) _returnvalue_6))
                            (:else)))))))
                (:else)))))))))
