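; Guided motion movement extension for SevenSlide that launches the slider off a ramp via a constant force.
; Caches the owner's velocity on activation and derives force direction and strength from a shared launch vector.
; That vector keeps current X/Y velocity and replaces Z with the VerticleSet scalable float.
; The cached starting velocity only fills a lateral-velocity local that does not affect the returned vector.
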(event EventOnExtensionActivate
  (params (RuntimeContext :type MMERuntime%20Context%20Structure%20%28by%20ref%29))
  (|CallParentFunction RuntimeContext)
  (Variables|Default|SetStartingVelocity (Variables|/Script/Engine.MovementComponent|GetVelocity (Variables|FortMovementMode_BaseExtLogic|GetOwnerMovementComponent))))

(event EventOnExtensionDeactivate
  (params (RuntimeContext :type MMERuntime%20Context%20Structure%20%28by%20ref%29))
  (|CallParentFunction RuntimeContext))

(fn GetDirection_ConstantForce
  (params (GuidedMotionRuntimeData :type Fort%20Movement%20Mode%20Guided%20Motion%20Runtime%20Data%20Object%20Reference))
  (bind _launchvector (CallFunction|GetLaunchVector))
  (return (Math|Vector|Normalize _launchvector)))

(fn GetLaunchVector
  (bind (_velocity_x _velocity_y) (Math|Vector|BreakVector (Variables|/Script/Engine.MovementComponent|GetVelocity_X (Variables|FortMovementMode_BaseExtLogic|GetOwnerMovementComponent))))
  (bind (_startingvelocity_x _startingvelocity_y) (Math|Vector|BreakVector (Variables|GetStartingVelocity_X)))
  (bind _output_get (LocalVariables|SetLateralVelocity (Math|Vector|MakeVector _startingvelocity_x _startingvelocity_y)))
  (return (Math|Vector|MakeVector _velocity_x _velocity_y (ScalableFloat|GetValueasDouble (Variables|Default|GetVerticleSet)))))

(fn GetStrength_ConstantForce
  (params (GuidedMotionRuntimeData :type Fort%20Movement%20Mode%20Guided%20Motion%20Runtime%20Data%20Object%20Reference))
  (bind _launchvector (CallFunction|GetLaunchVector))
  (return (Math|Vector|VectorLength _launchvector)))
